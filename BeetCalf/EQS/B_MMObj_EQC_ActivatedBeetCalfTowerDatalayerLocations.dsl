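; EQS context that supplies the locations of BeetCalf radio tower data layers that have already been activated.
; ProvideLocationsSet gets the Athena game state, finds its BeetCalf tower game state component, and returns that component's activated location list.
; Validity checks guard both lookups, with a fallback empty return when either is missing.
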(fn ProvideLocationsSet
  (params (QuerierObject :type Object%20Reference) (QuerierActor :type Actor%20Object%20Reference))
  (bind _returnvalue (Fortnite|Athena|GameState|GetGameStateAthena))
  (bind _returnvalue_2 (Actor|GetComponentbyClass _returnvalue "/BeetCalf/Gameplay/B_MMObj_BeetCalfTowerGameStateComponent.B_MMObj_BeetCalfTowerGameStateComponent_C"))
  ;[0.1366 0.03923 0.1771] Get locations of the radio towers that have already been activated.
  (Utilities|FlowControl|Sequence
    (:then_0
      (Utilities|IsValid _returnvalue
        (:Is%20Valid
          (Utilities|IsValid _returnvalue_2
            (:Is%20Valid
              (bind _locations (EQS|GetRadioTowerDatalayerLocationsActivated _returnvalue_2 :!class /BeetCalf/Gameplay/B_MMObj_BeetCalfTowerGameStateComponent.B_MMObj_BeetCalfTowerGameStateComponent_C))
              (return _locations))))))
    (:then_1
      (return 0))))
